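# Niagara impact effect for the PolkaSkate bass pickaxe, combining a flash, directional streaks, surface debris, fluttering petals and smoke.
# Burst emitters use spawn groups to vary layers within one emitter, including a core with sprite, mesh and light renderers.
# System direction and global scale drive motion and size, while time-of-day color tints the flash.
# Debris responds to the user-supplied physical surface, and front-end switches adjust several layers for menu display.
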NiagaraSystem /SparksCosmetics/Instrument/Bass/PolkaSkate/Effects/NS_Pickaxe_Bass_PolkaSkate_Impact
dependencies {
 usedDataInterfaces [
  &NiagaraDataInterfaceCurve
  &NiagaraDataInterfaceVector2DCurve
  &NiagaraDataInterfaceColorCurve
  &NiagaraDataInterfaceArrayFloat
  &NiagaraDataInterfaceArrayColor
  &NiagaraDataInterfaceSpriteRendererInfo
  &NiagaraDataInterfaceVectorCurve
  &NiagaraDataInterfaceArrayFloat3
 ]
 usedDynamicInputs [/Niagara/DynamicInputs/Math/Clamp_Vector, &SelectVectorFromArray]
 usedModules [
  &self:SystemSpawnScript.SystemScriptSource.SystemScriptGraph.NiagaraNodeAssignment_0.SetVariables_#1
  &SystemState
  &self:SystemSpawnScript.SystemScriptSource.SystemScriptGraph.NiagaraNodeAssignment_1.SetVariables_#2
 ]
}
emitters {
 Core_Impact_Burst {
  emitterUpdateScript [
   &EmitterState()
   &SpawnBurst_Instantaneous(
    Spawn Probability=&FN_FrontEndFloatModifier(Gameplay=0, Front End=1)
    Use Spawn Probability=true
    UseLoopCountLimit=false
   ) {enabled=false, module=SpawnBurst_Instantaneous001}
   &SpawnBurst_Instantaneous(Spawn Group=1, Use Spawn Probability=false, UseLoopCountLimit=false)
   &SpawnBurst_Instantaneous(
    Spawn Count=1
    Spawn Group=2
    Spawn Probability=&FN_FrontEndFloatModifier(Gameplay=0, Front End=1)
    Use Spawn Probability=true
   ) {enabled=false, module=SpawnBurst_Instantaneous002}
   &SpawnBurst_Instantaneous(
    Spawn Count=9
    Spawn Group=3
    Spawn Probability=&FN_FrontEndFloatModifier(Gameplay=0, Front End=1)
    Use Spawn Probability=true
   ) {enabled=false, module=SpawnBurst_Instantaneous003}
   &SpawnBurst_Instantaneous(Spawn Count=7, Spawn Group=4, Use Spawn Probability=true, UseLoopCountLimit=false) {module=SpawnBurst_Instantaneous004}
  ]
  particleSpawnScript [
   &PartitionParticles(Number Of Partitions=3)
   /SetVariables(Particles.VisibilityTag=&SpawnGroup()) {isSetParameters=true}
   &InitializeParticle(
    Lifetime=&MaskFloatBySpawnGroup(
     Masked Float=&FN_FrontEndFloatModifier(Gameplay=0.13, Front End=0.15)
     Passthrough Float=&UniformRangedFloat(Minimum=0.15, Maximum=0.2)
     Spawn Group Mask=1
    )
    Mesh Scale Mode=::NewEnumerator3
    Mesh Uniform Scale=0.35
    Sprite Rotation Mode=::NewEnumerator2
    Sprite Size=&MaskVector2dBySpawnGroup(
     Masked Vector2D=&UniformRangedVector2D(Minimum=<10 45>, Maximum=<12 65>)
     Passthrough Vector2D=&Vector2DFromFloat(Value=&Multiply_Float(A=120, B=$System.GlobalScale))
     Spawn Group Mask=4
    )
    Sprite Size Mode=::NewEnumerator4
    UsePositionOffset=false
    Write Material Random=true
   )
   &FN_PerParticleLOD(Enable SpriteSize Distance Scaling=true, Sprite Size Scale Far=<2.5 2.5>)
   &ShapeLocation(
    MaskSpawnGroup=false
    Sphere Distribution=::NewEnumerator3
    Sphere Radius=&MaskFloatBySpawnGroup(Masked Float=3, Passthrough Float=0, Spawn Group Mask=4)
    Uniform Spiral Amount=0.8
    UseOwnerScale=true
   )
   &AddVelocityFromPoint(
    Module.Use Falloff Distance=false
    Velocity Strength=&MaskFloatBySpawnGroup(
     Masked Float=&FloatFromCurve(
      FloatCurve=DataInterface() {propertyValues {Curve {keys [<at 0 0>,<at 1 15>]}}}
      CurveIndex=&RandomRangeFloat(Minimum=0, Maximum=1)
      Scale Curve=$System.GlobalScale
     )
     Passthrough Float=0
     Spawn Group Mask=4
    )
   )
   &AddVelocity(
    Velocity=$System.DirectionVector
    Velocity Speed Scale=&MaskFloatBySpawnGroup(
     Masked Float=&Multiply_Float(
      A=&RandomRangeFloat(Minimum=25, Maximum=150)
      B=&SelectFloatFromArray(Direct Array Index=$Particles.PartitionParticles.Partition, Float Selection Array=DataInterface() {propertyValues {FloatData [-1,1,1]}})
     )
     Passthrough Float=0
     Spawn Group Mask=4
    )
   )
   &InitialMeshOrientation(Orientation Axis=<0 0 1>, Orientation Vector=<0 0 1>, Rotation=<0 0.25 0>, Use Orientation Vector=true, Use Rotation Vector=true)
   &InitialMeshOrientation(
    Orientation Axis=<0 0 0>
    Orientation Vector=<0 0 0>
    Rotation=&RandomRangeVector(Minimum=<0 0 0>, Maximum=<1 0 0>)
    Use Orientation Vector=false
    Use Rotation Vector=true
   ) {module=InitialMeshOrientation001}
   /SetVariables(
    Particles.Flash Star Cam Offset=15
    Particles.Flash Star Lifetime=0.25
    Particles.Flash Star Sprite Size:Vector2f=&Vector2DFromFloat(Value=&FN_FrontEndFloatModifier(Gameplay=64, Front End=90))
    Particles.Flash Star SubUV Index:NiagaraFloat=&RandomRangeFloat(Minimum=0, Maximum=3)
   ) {isSetParameters=true}
   /SetVariables(
    Particles.FlareCardColor:LinearColor=$System.TOD Color
    Particles.FlareCardDynamicParameter:Vector4f=&MakeVector4(X=0.2, Y=1, Z=0, W=0)
   ) {isSetParameters=true}
   &CameraOffset(Camera Offset Amount=15)
   &ApplyInitialForces()
  ]
  particleUpdateScript [
   &ParticleState()
   &ScaleSpriteSize(
    Scale Factor=&MaskVector2dBySpawnGroup(
     Masked Vector2D=&Multiply_Vector2D_ByFloat(
      Float=&FloatFromCurve(FloatCurve=DataInterface() {propertyValues {Curve {
          defaultValue=1
          keys [
           {interpMode=RCIM_Cubic, tangentMode=RCTM_User, time=0, value=1}
           {arriveTangent=-3.0339, interpMode=RCIM_Cubic, leaveTangent=-3.0339, tangentMode=RCTM_User, time=1, value=0}
          ]
         }}})
      Vector2D=<1 1>
     )
     Passthrough Vector2D=&MakeVector2D(
      X=&FloatFromCurve(
       FloatCurve=DataInterface() {propertyValues {Curve {
          defaultValue=1
          keys [
           {arriveTangent=1.7844, interpMode=RCIM_Cubic, leaveTangent=1.7844, tangentMode=RCTM_User, time=0, value=0}
           {arriveTangentWeight=0.3771, interpMode=RCIM_Cubic, leaveTangentWeight=0.0943, tangentMode=RCTM_User, time=0.8, value=1}
           {arriveTangent=-5.8206, interpMode=RCIM_Cubic, leaveTangent=-5.8206, tangentMode=RCTM_User, time=1.0063, value=0.1176}
          ]
         }}}
       Scale Curve=$System.GlobalScale
      )
      Y=&FloatFromCurve(
       FloatCurve=DataInterface() {propertyValues {Curve {
          defaultValue=1
          keys [
           {arriveTangent=0.8542, interpMode=RCIM_Cubic, leaveTangent=0.8542, tangentMode=RCTM_User, time=0, value=0.3}
           {arriveTangent=3.2667, interpMode=RCIM_Cubic, leaveTangent=3.2667, tangentMode=RCTM_User, time=1, value=1.85}
          ]
         }}}
       Scale Curve=$System.GlobalScale
      )
     )
     Spawn Group Mask=1
    )
    Scale Sprite Size Mode=::NewEnumerator2
   )
   &LinearForce(Force=&Multiply_VectorByFloat(
     Vector=$System.DirectionVector
     Float=&MaskFloatBySpawnGroup(
      Masked Float=&SelectFloatFromArray(
       Direct Array Index=$Particles.PartitionParticles.Partition
       Float Selection Array=DataInterface() {propertyValues {FloatData [25,3900,4500]}}
      )
      Passthrough Float=0
      Spawn Group Mask=4
     )
    ))
   &ScaleMeshSize(Scale Factor=&VectorFromFloat(Value=&FloatFromCurve(
      FloatCurve=DataInterface() {propertyValues {Curve {keys [
          {arriveTangent=7.2985, interpMode=RCIM_Cubic, leaveTangent=7.2985, tangentMode=RCTM_User, time=0, value=0.5}
          {arriveTangent=0.6472, interpMode=RCIM_Cubic, leaveTangent=0.6472, tangentMode=RCTM_User, time=0.15, value=1.25}
          {arriveTangent=1.8898, interpMode=RCIM_Cubic, leaveTangent=1.8898, tangentMode=RCTM_User, time=1, value=2}
         ]}}}
      Scale Curve=$System.GlobalScale
     )))
   &ScaleColor(
    Scale Alpha=&FloatFromCurve(
     FloatCurve=DataInterface() {propertyValues {Curve {
        defaultValue=1
        keys [
         {arriveTangent=-1.5437, interpMode=RCIM_Cubic, leaveTangent=-1.5437, tangentMode=RCTM_User, time=0, value=1}
         {interpMode=RCIM_Cubic, tangentMode=RCTM_User, time=1, value=0}
        ]
       }}}
     Scale Curve=1.1
    )
    Scale RGB=&VectorFromFloat(Value=1)
    ScaleA=true
    ScaleRGB=true
   )
   /SetVariables(Particles.DynMesh:Vector4f=&MakeVector4(
     X=&FloatFromCurve(FloatCurve=DataInterface() {propertyValues {Curve {keys [
          {arriveTangent=0.7731, interpMode=RCIM_Cubic, leaveTangent=0.7731, leaveTangentWeight=0.348, tangentMode=RCTM_User, time=0, value=0}
          {arriveTangentWeight=0.3648, interpMode=RCIM_Cubic, tangentMode=RCTM_User, time=1, value=0.25}
         ]}}})
     Y=&FloatFromCurve(FloatCurve=DataInterface() {propertyValues {Curve {keys [<at 0 0.3>,<at 1 1.5>]}}})
     Z=1
     W=&FloatFromCurve(FloatCurve=DataInterface() {propertyValues {Curve {keys [{arriveTangent=-0.4202, interpMode=RCIM_Cubic, leaveTangent=-0.4202, tangentMode=RCTM_User, time=0.4, value=1}, <at 1 0.7 RCIM_Cubic>]}}})
    )) {enabled=false, isSetParameters=true}
   /SetVariables(Particles.Color_Mesh:LinearColor=&MultiplyLinearColor(
     Linear Color=<rgba 0 0 0 1.1>
     Scale Factor=&MakeFloat4FromVectorAndFloat(
      "Vector (RGB)"=&VectorFromFloat(Value=1)
      "Float (Alpha)"=&FloatFromCurve(
       FloatCurve=DataInterface() {propertyValues {Curve {
          defaultValue=0
          keys [
           {interpMode=RCIM_Cubic, tangentMode=RCTM_User, time=0, value=1}
           {arriveTangentWeight=0.7106, interpMode=RCIM_Cubic, tangentMode=RCTM_User, time=1, value=0}
          ]
         }}}
       Scale Curve=&FloatFromCurve(
        FloatCurve=DataInterface() {propertyValues {Curve {keys [<at 0 1>,<at 0.3335 0.8>,<at 1 1 RCIM_Cubic>]}}}
        CurveIndex=Linked() {linkedVariable {name=NPC.FortniteNPC.TimeOfDay_0-1, type=NiagaraFloat}}
        Scale Curve=&MakeCustomFloatFromBool(Bool=$User.bIsFrontEnd, True Float=0.65, False Float=1)
       )
      )
     )
    )) {isSetParameters=true}
   /SetVariables(Particles.Update Flash Star Age:NiagaraFloat=&Divide_Float(A=$Particles.Age, B=$Particles.Initial.Flash Star Lifetime)) {isSetParameters=true}
   /SetVariables(
    Particles.Flash Star Color:LinearColor=&MultiplyLinearColor(
     Linear Color=&ColorFromCurve(
      ColorCurve=DataInterface() {propertyValues {
        AlphaCurve {keys [<at 0 1>,<at 1 0>]}
        BlueCurve {keys [<at 0.1395 6>,<at 0.4913 0.9378>,<at 0.8637 0.0021>]}
        GreenCurve {keys [<at 0.1395 6>,<at 0.4913 0.9211>,<at 0.8637 0.0028>]}
        RedCurve {keys [<at 0.1395 6>,<at 0.4913 2>,<at 0.8637 0.02>]}
       }}
      CurveIndex=$Particles.Update Flash Star Age
     )
     Scale Factor=&MakeVector4_FromLinearColor(LinearColor=$System.TOD Color)
    )
    Particles.Update Flash Star Sprize Size:Vector2f=&Multiply_Vector2D_ByFloat(
     Float=&FloatFromCurve(
      FloatCurve=DataInterface() {propertyValues {Curve {
         defaultValue=0.2
         keys [
          {arriveTangent=1.826, interpMode=RCIM_Cubic, leaveTangent=1.826, leaveTangentWeight=1.0797, tangentMode=RCTM_User, time=0.0036, value=0.2824}
          {arriveTangentWeight=0.6623, interpMode=RCIM_Cubic, leaveTangentWeight=0.3962, tangentMode=RCTM_User, time=0.627, value=0.65}
          {arriveTangent=6.378, arriveTangentWeight=0.4894, interpMode=RCIM_Cubic, leaveTangent=6.378, tangentMode=RCTM_User, time=1, value=1.5}
         ]
        }}}
      CurveIndex=$Particles.Update Flash Star Age
      Scale Curve=$System.GlobalScale
     )
     Vector2D=$Particles.Initial.Flash Star Sprite Size
    )
    Particles.UpdateFlareCardColor:LinearColor=&MakeLinearColorFromVectorAndFloat(
     "Vector (RGB)"=&Multiply_Vector(
      A=&Multiply_Vector(
       A=&MakeVectorFromLinearColorRGB(Color=$Particles.Initial.FlareCardColor)
       B=&MakeVectorFromLinearColorRGB(Color=&ColorFromCurve(ColorCurve=DataInterface() {propertyValues {
           AlphaCurve {keys [<at 0 1 RCIM_Cubic>,<at 1 0 RCIM_Cubic>]}
           BlueCurve {keys [<at 0 4>,<at 1 0.1596>]}
           GreenCurve {keys [<at 0 2.3291>,<at 1 0.0006>]}
           RedCurve {keys [<at 0 2.8571>,<at 1 0.0686>]}
          }}))
      )
      B=&VectorFromFloat(Value=&FN_FrontEndFloatModifier(Gameplay=1.5, Front End=1))
     )
     "Float (Alpha)"=&FloatFromCurve(
      FloatCurve=DataInterface() {propertyValues {Curve {keys [{arriveTangent=-1.9727, interpMode=RCIM_Cubic, time=0, value=1}, <at 1 0 RCIM_Cubic>]}}}
      Scale Curve=&MakeCustomFloatFromBool(Bool=$User.bIsFrontEnd, True Float=0.5, False Float=1)
     )
    )
   ) {isSetParameters=true}
   /SetVariables(Particles.ColorDirectionalSpikes:LinearColor=&MultiplyLinearColor(
     Linear Color=&ColorFromCurve(ColorCurve=DataInterface() {propertyValues {
        AlphaCurve {keys [<at 0.2866 1>,<at 1 0>]}
        BlueCurve {keys [<at 0.0768 4>,<at 0.379 0.7034>,<at 0.8274 0.0021>]}
        GreenCurve {keys [<at 0.0768 4>,<at 0.379 0.6908>,<at 0.8274 0.0028>]}
        RedCurve {keys [<at 0.0768 4>,<at 0.379 1.5>,<at 0.8274 0.02>]}
       }})
     Scale Factor=&MakeVector4_FromLinearColor(LinearColor=$System.TOD Color)
    )) {isSetParameters=true}
   &Drag(
    Drag=&MaskFloatBySpawnGroup(Masked Float=1, Passthrough Float=2, Spawn Group Mask=4)
    Use Linear Drag=true
    Use Rotational Drag=true
   )
   &SolveForcesAndVelocity(Clamp Velocity=false, Limit Acceleration=false)
   &FN_Light_Attributes(
    Color=$System.TOD Color
    Create Color Attribute=true
    Create Exponent Attribute=true
    Create Position Attribute=true
    Create Radius Attribute=true
    Exponent=4
    Position=<0 5 0>
    Radius=90
    Scale Color=&FloatFromCurve(
     FloatCurve=DataInterface() {propertyValues {Curve {keys [
         {interpMode=RCIM_Cubic, tangentMode=RCTM_User, time=0, value=1}
         {arriveTangent=-0.5426, interpMode=RCIM_Cubic, leaveTangent=-0.5426, tangentMode=RCTM_User, time=0.8, value=0}
        ]}}}
     Scale Curve=15
    )
   )
  ]
  rendererClasses [
   &NiagaraSpriteRendererProperties
   &NiagaraLightRendererProperties
   &NiagaraMeshRendererProperties
  ]
  renderers [
   {rendererClass=&NiagaraSpriteRendererProperties}
   {rendererClass=&NiagaraLightRendererProperties, rendererIndex=1}
   {rendererClass=&NiagaraSpriteRendererProperties, rendererIndex=2}
   {rendererClass=&NiagaraMeshRendererProperties, rendererIndex=3}
   {rendererClass=&NiagaraSpriteRendererProperties, rendererIndex=4}
   {rendererClass=&NiagaraSpriteRendererProperties, rendererIndex=5}
  ]
 }
 Directional_Elements {
  emitterSpawnScript [/SetVariables(Emitter.SpawnCount=2) {isSetParameters=true}]
  emitterUpdateScript [
   &EmitterState()
   &SpawnBurst_Instantaneous(Spawn Count=2, Use Spawn Probability=false, UseLoopCountLimit=false)
   &SpawnBurst_Instantaneous(Spawn Count=3, Spawn Group=1, Use Spawn Probability=false, UseLoopCountLimit=false) {module=SpawnBurst_Instantaneous001}
   &SpawnBurst_Instantaneous(Spawn Group=2, Use Spawn Probability=false, UseLoopCountLimit=false) {module=SpawnBurst_Instantaneous002}
  ]
  particleSpawnScript [
   /SetVariables(Particles.Variation:NiagaraFloat=&FloatFromCurve(
     FloatCurve=DataInterface() {propertyValues {Curve {keys [{tangentMode=RCTM_User, time=0, value=0}, {tangentMode=RCTM_User, time=1, value=1}]}}}
     CurveIndex=&Divide_Float(
      A=&MakeFloatFromInt(Int=&ReturnExecIndex())
      B=&MakeFloatFromInt(Int=&Add_Integer(A=$Emitter.SpawnCount, B=-1))
     )
    )) {isSetParameters=true}
   /SetVariables(Particles.VisibilityTag=&SpawnGroup()) {isSetParameters=true}
   &InitializeParticle(
    Color=&RandomRangeLinearColor(
     Minimum=&SelectLinearColorFromArray(
      Array Sampling Mode=::NewEnumerator1
      Direct Array Index=&SpawnGroup()
      Color Selection Array=DataInterface() {propertyValues {ColorData [<rgba 3 3 3 1>, <rgba 2 1.065 1.1571 1>, <rgba 1 0.6134 0.5559 1>]}}
      Random Seed=?
     )
     Maximum=&SelectLinearColorFromArray(
      Array Sampling Mode=::NewEnumerator1
      Direct Array Index=&SpawnGroup()
      Color Selection Array=DataInterface() {propertyValues {ColorData [<rgba 1.5 1.5 1.5 1>, <rgba 2 1.065 1.1571 1>, <rgba 1 0.6134 0.5559 1>]}}
      Random Seed=?
     )
    )
    Lifetime Max=&SelectFloatFromArray(
     Direct Array Index=&SpawnGroup()
     Float Selection Array=DataInterface() {propertyValues {FloatData [0.19,0.18,0.18]}}
    )
    Lifetime Min:NiagaraFloat=&SelectFloatFromArray(
     Direct Array Index=&SpawnGroup()
     Float Selection Array=DataInterface() {propertyValues {FloatData [0.22,0.12,0.18]}}
    )
    Lifetime Mode=::NewEnumerator1
    Sprite Size=&MaskVector2dBySpawnGroup(
     Masked Vector2D=&Multiply_Vector2D_ByFloat(Float=&RandomRangeFloat(Minimum=55, Maximum=90), Vector2D=<1 1>)
     Passthrough Vector2D=&MaskVector2dBySpawnGroup(
      Masked Vector2D=&UniformRangedVector2D(Minimum=<15 60>, Maximum=<20 80>)
      Passthrough Vector2D=&Multiply_Vector2D_ByFloat(Float=0.8, Vector2D=<80 240>)
      Spawn Group Mask=1
     )
     Spawn Group Mask=0
    )
    Sprite Size Mode=::NewEnumerator4
    UsePositionOffset=false
    Write Material Random=true
   )
   &ShapeLocation(
    MaskSpawnGroup=false
    Override Seed=false
    Sphere Radius=&MaskFloatBySpawnGroup(Masked Float=6, Passthrough Float=0)
    UseOwnerScale=true
   )
   &AddVelocity(
    Velocity=$System.DirectionVector
    Velocity Speed Scale=&MaskFloatBySpawnGroup(
     Masked Float=&Multiply_Float(
      A=&RandomRangeFloat(Minimum=300, Maximum=500)
      B=&FN_FrontEndFloatModifier(Gameplay=1, Front End=1)
     )
     Passthrough Float=&MaskFloatBySpawnGroup(
      Masked Float=&Multiply_Float(A=100, B=$System.GlobalScale)
      Passthrough Float=&Multiply_Float(A=300, B=$System.GlobalScale)
      Spawn Group Mask=1
     )
    )
   ) {module=AddVelocity001}
   /SetVariables(Particles.LineSubUV:NiagaraFloat=&MakeFloatFromInt(Int=&RandomRangeInt(Minimum=0, Maximum=3))) {isSetParameters=true}
   &SubUVAnimation(
    End Frame Range Override=3
    Output Alternate Renderer Binding=true
    "Sprite Renderer "=DataInterface() {propertyValues {SpriteRenderer=&self:Directional_Elements_0.NiagaraSpriteRendererProperties_0}}
    SubUV Animation Mode:ENiagara_SubUVLookupModeV2=::NewEnumerator2
    UseEndFrame=true
    UseStartFrame=true
   )
   &CameraOffset(Camera Offset Amount=&MaskFloatBySpawnGroup(Masked Float=30, Passthrough Float=5, Spawn Group Mask=2))
   /SetVariables(
    Particles.DynFlare0:Vector4f=&MakeVector4(X=0.2, Y=1, Z=0, W=0)
    Particles.DynFlare3:Vector4f=&MakeVector4(X=1.5, Y=1, Z=0, W=0)
   ) {isSetParameters=true}
  ]
  particleUpdateScript [
   &ParticleState()
   &Drag(
    Drag=&MaskFloatBySpawnGroup(Masked Float=&FN_FrontEndFloatModifier(Gameplay=2.5, Front End=5), Passthrough Float=0)
    Use Linear Drag=true
    Use Rotational Drag=false
   )
   &SolveForcesAndVelocity(Clamp Velocity=false, Limit Acceleration=false)
   &DynamicMaterialParameters(
    Index 0 Param 1=&FloatFromCurve(FloatCurve=DataInterface() {propertyValues {Curve {keys [
         {arriveTangent=0.8115, interpMode=RCIM_Cubic, leaveTangent=0.8115, tangentMode=RCTM_User, time=0.25, value=0.2}
         {arriveTangent=0.7921, interpMode=RCIM_Cubic, leaveTangent=0.7921, tangentMode=RCTM_User, time=1, value=1}
        ]}}})
    Index 0 Param 2=&SelectFloatFromArray(
     Array Sampling Mode=::NewEnumerator0
     Direct Array Index=?
     Float Selection Array=DataInterface() {propertyValues {FloatData [0,3]}}
     Random Seed=0
    )
    Index 1 Param 2=1
    Index 1 Param 3=1
    Index 1 Param 4=&FloatFromCurve(FloatCurve=DataInterface() {propertyValues {Curve {keys [<at 0.5 1 RCIM_Cubic>, {arriveTangent=-3.1839, tangentMode=RCTM_User, time=1, value=0.35}]}}})
    Module.Index0Float0Write=true
    Module.Index0Float1Write=true
    Module.Index0Float2Write=true
    Module.Index0Float3Write=true
    Module.Index1Float0Write=true
    Module.Index1Float1Write=true
    Module.Index1Float2Write=true
    Module.Index1Float3Write=true
    Write Parameter Index 1=::NewEnumerator1
   )
   &ScaleColor(
    Scale Alpha=&MaskFloatBySpawnGroup(
     Masked Float=&FloatFromCurve(
      FloatCurve=DataInterface() {propertyValues {Curve {keys [
          {arriveTangent=7.213, interpMode=RCIM_Cubic, leaveTangent=7.213, tangentMode=RCTM_User, time=0, value=0.6}
          {arriveTangentWeight=0.0568, interpMode=RCIM_Cubic, leaveTangentWeight=0.1099, tangentMode=RCTM_User, time=0.1023, value=1}
          {interpMode=RCIM_Cubic, leaveTangentWeight=0.3967, tangentMode=RCTM_User, time=0.5, value=1}
          {arriveTangent=-3.6906, interpMode=RCIM_Cubic, leaveTangent=-3.6906, tangentMode=RCTM_User, time=1, value=0}
         ]}}}
      Scale Curve=&FN_FrontEndFloatModifier(Gameplay=1, Front End=0.2)
     )
     Passthrough Float=&MaskFloatBySpawnGroup(
      Masked Float=&FloatFromCurve(FloatCurve=DataInterface() {propertyValues {Curve {keys [<at 0 1>,<at 1 0>]}}})
      Passthrough Float=1
      Spawn Group Mask=1
     )
    )
    Scale RGB=&MaskVectorBySpawnGroup(
     Masked Vector=&MakeVectorFromLinearColorRGB(Color=&ColorFromCurve(ColorCurve=DataInterface() {propertyValues {
         AlphaCurve {keys [<at 0 1 RCIM_Cubic>]}
         BlueCurve {keys [<at 0.0101 3.9444>,<at 0.1864 0.44>,<at 0.4777 0>]}
         GreenCurve {keys [<at 0.0101 3.269>,<at 0.1864 0.4524>,<at 0.4777 0.0002>]}
         RedCurve {keys [<at 0.0101 3.547>,<at 0.1864 0.5651>,<at 0.4777 0.014>]}
        }}))
     Passthrough Vector=&MaskVectorBySpawnGroup(
      Masked Vector=&MakeVectorFromLinearColorRGB(Color=&ColorFromCurve(ColorCurve=DataInterface() {propertyValues {
          AlphaCurve {keys [<at 0 1 RCIM_Cubic>]}
          BlueCurve {keys [<at 0.1131 5>,<at 0.5526 0.1619>,<at 0.7359 0>]}
          GreenCurve {keys [<at 0.1131 3.9111>,<at 0.5526 0.1428>,<at 0.7359 0.001>]}
          RedCurve {keys [<at 0.1131 4.3593>,<at 0.5526 0.1901>,<at 0.7359 0.01>]}
         }}))
      Passthrough Vector=<1 1 1>
      Spawn Group Mask=1
     )
    )
    ScaleA=true
    ScaleRGB=true
   )
   &ScaleSpriteSize(Scale Factor=&MaskVector2dBySpawnGroup(
     Masked Vector2D=&Vector2DFromCurve(
      Vector2Curve=DataInterface() {propertyValues {XCurve {keys [<at 0 0.6 RCIM_Cubic>,<at 1 0 RCIM_Cubic>]}, YCurve {keys [<at -0.0039 0.1279 RCIM_Cubic>,<at 1 2 RCIM_Cubic>]}}}
      ScaleCurve=&Multiply_Vector2D_ByFloat(Float=$System.GlobalScale, Vector2D=<1.5 1>)
     )
     Passthrough Vector2D=&Vector2DFromCurve(
      Vector2Curve=DataInterface() {propertyValues {
        XCurve {keys [{tangentMode=RCTM_User, time=0, value=0}, <at 0.2 1.075>, {tangentMode=RCTM_User, time=1, value=-0.1}]}
        YCurve {keys [{tangentMode=RCTM_User, time=0, value=0}, <at 0.1 1.025>, {tangentMode=RCTM_User, time=1, value=0.8}]}
       }}
      ScaleCurve=<1 1>
     )
     Spawn Group Mask=1
    ))
   /SetVariables(
    Transient.DistanceScale:NiagaraFloat=&FloatFromCurve(
     FloatCurve=DataInterface() {propertyValues {Curve {keys [<at 0 1 RCIM_Cubic>, {arriveTangent=-2.2435, tangentMode=RCTM_User, time=1, value=0.1}]}}}
     CurveIndex=&Multiply_Float(A=$Particles.NormalizedAge, B=&RandomRangeFloat(Minimum=0.75, Maximum=1))
     Scale Curve=&Multiply_Float(
      A=&Multiply_Float(A=2.75, B=$System.GlobalScale)
      B=&VectorLength(Vector=&Subtract_Vector(
        A=Linked() {linkedVariable {name=Particles.Position, type=NiagaraPosition}}
        B=Linked() {linkedVariable {name=Particles.Initial.Position, type=NiagaraPosition}}
       ))
     )
    )
    Transient.WidthScale:NiagaraFloat=&FloatFromCurve(
     FloatCurve=DataInterface() {propertyValues {Curve {keys [<at 0 0.3>,<at 1 0.2>]}}}
     Scale Curve=&Multiply_Float(A=&MakeFloatFromVector2d(Vector 2D=$Particles.Initial.SpriteSize), B=$System.GlobalScale)
    )
   ) {isSetParameters=true}
   /SetVariables(
    Particles.LineColor:LinearColor=&MultiplyLinearColor(Linear Color=$Particles.Color, Scale Factor=<1 1 1 1>)
    Particles.LineScale:Vector2f=&MakeVector2D(X=$Transient.WidthScale, Y=$Transient.DistanceScale)
   ) {isSetParameters=true}
   &SpriteFacingAndAlignment(
    Sprite Alignment=&Multiply_VectorByFloat(Vector=$Particles.Velocity, Float=-1)
    WriteAlignment=true
    WriteFacing=false
   )
  ]
  rendererClasses [&NiagaraSpriteRendererProperties]
  renderers [
   {rendererClass=&NiagaraSpriteRendererProperties}
   {rendererClass=&NiagaraSpriteRendererProperties, rendererIndex=1}
   {rendererClass=&NiagaraSpriteRendererProperties, rendererIndex=2}
  ]
 }
 FN_PhysMatDebris {
  emitterUpdateScript [
   &EmitterState()
   &SpawnBurst_Instantaneous(
    Spawn Count=&RandomRangeInt(Minimum=5, Maximum=10)
    Use Spawn Probability=false
    UseLoopCountLimit=false
   )
  ]
  particleSpawnScript [
   /SetVariables(Particles.BiasedRandom=&FloatFromCurve(
     FloatCurve=DataInterface() {propertyValues {Curve {keys [
         <at 0.2 1 RCIM_Cubic>
         {arriveTangent=-3.3394, interpMode=RCIM_Cubic, leaveTangent=-3.3394, tangentMode=RCTM_User, time=0.5, value=0.5}
         <at 0.8 0 RCIM_Cubic>
        ]}}}
     CurveIndex=&RandomRangeFloat(Minimum=0, Maximum=1)
    )) {isSetParameters=true}
   &InitializeParticle(
    Color Mode=::NewEnumerator0
    Lifetime=&RandomRangeFloat(Minimum=0.5, Maximum=1.25)
    Mass=&Lerp_Float(A=0.6, B=1.2, Alpha=$Particles.BiasedRandom)
    Mass Mode=::NewEnumerator1
    Mesh Scale=&Multiply_VectorByFloat(
     Vector=&RandomRangeVector(Minimum=<0.9 0.4 0.5>, Maximum=<1 1 1>)
     Float=&Lerp_Float(A=0.7, B=1, Alpha=$Particles.BiasedRandom)
    )
    Mesh Scale Mode=::NewEnumerator4
    UsePositionOffset=false
    Write Material Random=true
   )
   &FN_PhysicalSurface(
    Dirt Mesh Count=2
    Glass Mesh Count=3
    Ice Mesh Count=2
    Metal Gravity=&RandomRangeFloat(Minimum=-350, Maximum=-1800)
    Metal Mesh Count=2
    Metal Velocity=&RandomRangeFloat(Minimum=80, Maximum=375)
    Metal Velocity Cone Angle=130
    Physical Surface=$User.PhysicalSurface
    Stone Gravity=&RandomRangeFloat(Minimum=-700, Maximum=-1250)
    Stone Mesh Count=2
    Stone Velocity=&RandomRangeFloat(Minimum=80, Maximum=550)
    Stone Velocity Cone Angle=120
    Wood Gravity=&RandomRangeFloat(Minimum=-450, Maximum=-900)
    Wood Mesh Count=2
    Wood Scale=1.5
    Wood Velocity=&RandomRangeFloat(Minimum=80, Maximum=600)
    Wood Velocity Cone Angle=130
   )
   &AddVelocityInCone(
    Cone Angle=$Output.FN_PhysicalSurface.SurfaceResponseConeAngle
    Module.Use Velocity Falloff On Cone Axis=false
    Velocity Strength=$Output.FN_PhysicalSurface.SurfaceResponseVelocity
   )
   &InitialMeshOrientation(Use Rotation Vector=false)
   &AddRotationalVelocity(Rotation Rate=&RandomVector(Vector Scale=&RandomRangeFloat(Minimum=1, Maximum=2)))
   &ApplyInitialForces()
   /SetVariables(Particles.GravityOverrideValue=$Output.FN_PhysicalSurface.GravityOverride) {isSetParameters=true}
  ]
  particleUpdateScript [
   &ParticleState()
   &GravityForce(Gravity=$Particles.GravityOverrideValue)
   &Collision(
    Advanced Aging Rate=1
    Age Colliding Particles=false
    CPU Collision Trace Channel=ECC_WorldStatic
    Collision Enabled=&Boolean_AndOperation(
     Bool1=&SetBoolByFloatComparison(A=$Particles.Initial.Lifetime, B=1)
     Bool2=&SetBoolByFloatComparison(A=$Particles.Age, B=0.12)
    )
    Control Roll On Collision=true
    Friction=1
    Friction During a Bounce=0.25
    Max CPU Trace Length=100
    Module.Control Rest State Via Surface Normal=false
    Module.EnableMaxCollisionCount=false
    Module.Randomize Collision Normal Vector=true
    Randomize Collision Normal=0.25
    ScaleRateOfRotation=false
   ) {enabled=false}
   &SolveForcesAndVelocity(Clamp Velocity=false, Limit Acceleration=false)
   &ScaleMeshSize(Scale Factor=&VectorFromFloat(Value=&FloatFromCurve(FloatCurve=DataInterface() {propertyValues {Curve {keys [<at 0.6 1 RCIM_Cubic>, {arriveTangent=-6.308, interpMode=RCIM_Cubic, leaveTangent=-6.308, tangentMode=RCTM_User, time=1, value=0}]}}})))
   &DynamicMaterialParameters(Module.Param0WriteEnabled=true, Module.Param1WriteEnabled=true, Module.Param2WriteEnabled=false, Module.Param3WriteEnabled=false)
  ]
  rendererClasses [&NiagaraSpriteRendererProperties, &NiagaraMeshRendererProperties]
  renderers [
   {rendererClass=&NiagaraSpriteRendererProperties}
   {rendererClass=&NiagaraMeshRendererProperties, rendererIndex=1}
   {rendererClass=&NiagaraMeshRendererProperties, rendererIndex=2}
   {rendererClass=&NiagaraMeshRendererProperties, rendererIndex=3}
   {rendererClass=&NiagaraMeshRendererProperties, rendererIndex=4}
   {rendererClass=&NiagaraMeshRendererProperties, rendererIndex=5}
   {rendererClass=&NiagaraMeshRendererProperties, rendererIndex=6}
   {rendererClass=&NiagaraMeshRendererProperties, rendererIndex=7}
   {rendererClass=&NiagaraMeshRendererProperties, rendererIndex=8}
  ]
 }
 Petals {
  emitterUpdateScript [
   &EmitterState(Life Cycle Mode=::NewEnumerator1, Loop Behavior=::NewEnumerator1, Loop Duration=1, UseLoopDelay=false)
   &SpawnBurst_Instantaneous(
    Spawn Count=&RandomRangeInt(Minimum=5, Maximum=7)
    Use Spawn Probability=false
    UseLoopCountLimit=false
   )
   /SetVariables(
    Emitter.WindOscillateTime:NiagaraFloat=&Divide_Float(A=$Engine.Time, B=20)
    Emitter.WindRotateTime:NiagaraFloat=&Divide_Float(A=$Engine.Time, B=100)
   ) {enabled=false, isSetParameters=true}
   /SetVariables(
    Emitter.WindIntensity:NiagaraFloat=&Multiply_Float(A=&Sine(Normalized Angle=$Emitter.WindOscillateTime, Period=1, Scale=1, Bias=0), B=0)
    Emitter.WindVector:Vector3f=&MakeVector(
     X=&Add_Float(
      A=&Sine(Normalized Angle=$Emitter.WindRotateTime, Period=1, Scale=1, Bias=0)
      B=&Cosine(Normalized Angle=$Emitter.WindRotateTime, Period=1, Scale=1, Bias=0)
     )
     Y=&Subtract_Float(
      A=&Cosine(Normalized Angle=$Emitter.WindRotateTime, Period=1, Scale=1, Bias=0)
      B=&Sine(Normalized Angle=$Emitter.WindRotateTime, Period=1, Scale=1, Bias=0)
     )
     Z=0
    )
   ) {enabled=false, isSetParameters=true}
  ]
  particleSpawnScript [
   &InitializeParticle(
    Color=&SelectLinearColorFromArray(
     Array Sampling Mode=::NewEnumerator0
     Direct Array Index=?
     Color Selection Array=DataInterface() {propertyValues {ColorData [<rgba 0.015 0 0.0002 1>, <rgba 0.02 0.0019 0 1>]}}
     Random Seed=0
    )
    Lifetime Max=1.1
    Lifetime Min=0.7
    Lifetime Mode=::NewEnumerator1
    Mass=&Multiply_Float(
     A=&RandomRangeFloat(Minimum=2, Maximum=4)
     B=&FloatFromCurve(
      FloatCurve=DataInterface() {propertyValues {Curve {keys [<at 0 0.5 RCIM_Cubic>, {arriveTangent=1.5048, interpMode=RCIM_Cubic, leaveTangent=1.5048, tangentMode=RCTM_User, time=1, value=1}]}}}
      CurveIndex=&ReturnNormalizedExecIndex()
      Scale Curve=1.5
     )
    )
    Mass Mode=::NewEnumerator1
    Sprite Rotation Mode=::NewEnumerator2
    Sprite Size Max=<8 6>
    Sprite Size Min=<4 2>
    Sprite Size Mode=::NewEnumerator2
    UsePositionOffset=false
    Write Material Random=true
   )
   &ShapeLocation(
    MaskSpawnGroup=false
    Override Seed=false
    Sphere Radius=&Multiply_Float(A=3, B=$System.GlobalScale)
    UseOwnerScale=true
   )
   &AddVelocity(
    Module.Use Falloff Distance=false
    Origin Offset=<0 0 -2>
    Origin Offset Coordinate Space=::World
    Override Seed=false
    Velocity Mode=::NewEnumerator1
    Velocity Speed=&Multiply_Float(
     A=&RandomRangeFloat(Minimum=190, Maximum=325)
     B=&FN_FrontEndFloatModifier(Gameplay=1, Front End=0.5)
    )
   )
   &AddVelocity(
    Velocity=$System.DirectionVector
    Velocity Speed Scale=&Multiply_Float(
     A=&Multiply_Float(
      A=&MakeCustomFloatFromBool(Bool=$User.bIsFrontEnd, True Float=0, False Float=125)
      B=&SelectFloatFromArray(
       Array Sampling Mode=::NewEnumerator0
       Direct Array Index=?
       Float Selection Array=DataInterface() {propertyValues {FloatData [0.25,1,0,0]}}
       Random Seed=0
      )
     )
     B=$System.GlobalScale
    )
   ) {module=AddVelocity001}
   &InitialMeshOrientation(Use Rotation Vector=false)
   &CameraOffset(Camera Offset Amount=15)
  ]
  particleUpdateScript [
   &ParticleState()
   &ScaleSpriteSize(
    Scale Factor=&Multiply_Vector2(
     A=&MakeVector2D(
      X=&Multiply_Float(
       A=1
       B=&Sine(
        Normalized Angle=$Emitter.Age
        Period=&Multiply_Float(A=3, B=&Add_Float(A=$Particles.Initial.MaterialRandom, B=0.6))
        Scale=1
        Bias=0
       )
      )
      Y=1
     )
     B=&Vector2DFromFloat(Value=&FloatFromCurve(
       FloatCurve=DataInterface() {propertyValues {Curve {keys [
           {arriveTangent=16.7921, interpMode=RCIM_Cubic, leaveTangent=16.7921, tangentMode=RCTM_User, time=0, value=0.5}
           {arriveTangent=0.0073, interpMode=RCIM_Cubic, leaveTangent=0.0073, tangentMode=RCTM_User, time=0.1, value=1}
           {arriveTangent=-1.2169, interpMode=RCIM_Cubic, leaveTangent=-1.2169, tangentMode=RCTM_User, time=0.6588, value=0.7302}
           <at 1 0.25>
          ]}}}
       Scale Curve=&Multiply_Float(A=3, B=$System.GlobalScale)
      ))
    )
    Scale Sprite Size Mode=::NewEnumerator2
   ) {module=ScaleSpriteSize001}
   &Drag(Drag=2, Use Linear Drag=true, Use Rotational Drag=true)
   &GravityForce(Gravity=&RandomRangeVector(Minimum=<0 0 -750>, Maximum=<0 0 -980>)) {module=GravityForce001}
   &AccelerationForce(Acceleration=&RandomRangeVector(Minimum=<0 0 33>, Maximum=<0 0 66>))
   &WindForce(Wind Speed=$Emitter.WindVector, Wind Speed Scale=$Emitter.WindIntensity) {enabled=false}
   &AerodynamicDrag(
    Aerodynamic Drag:NiagaraFloat=&FloatFromCurve(
     FloatCurve=DataInterface() {propertyValues {Curve {
        defaultValue=1
        keys [
         {arriveTangent=2.3817, interpMode=RCIM_Cubic, leaveTangent=2.3817, tangentMode=RCTM_User, time=0.1, value=0}
         {interpMode=RCIM_Cubic, tangentMode=RCTM_User, time=0.8, value=1}
        ]
       }}}
     Scale Curve=&RandomRangeFloat(Minimum=1.5, Maximum=2.5)
    )
    Aerodynamic Rotational Drag=0.1
    Lift Contribution:NiagaraFloat=&RandomRangeFloat(Minimum=0.3, Maximum=1)
    Pivot Offset Maximum:Vector3f=<4 0.75 0.25>
    Pivot Offset Minimum:Vector3f=<-4 -0.75 -0.25>
    Rotation Strength:NiagaraFloat=&RandomRangeFloat(Minimum=0.75, Maximum=2)
    Use Linear Drag=true
    Use Rotational Drag=true
   )
   &PointForce(
    Force Origin=&AddVectorToPosition(Position=/Niagara/DynamicInputs/Helpers/SimulationPosition(), Vector=<0 0 50>)
    Force Strength=&FN_FrontEndFloatModifier(Gameplay=1000, Front End=0)
    Module.Use Falloff=true
    Module.UseLinearForce=false
   )
   &SolveForcesAndVelocity(Clamp Velocity=false, Limit Acceleration=false)
   &AlignSpriteToMeshOrientation(SetAlignment=true, SetFacing=true)
   &DynamicMaterialParameters(
    Index 0 Param 1=&RandomRangeFloat(Minimum=0, Maximum=0.5)
    Module.Index0Param0Write=true
    Module.Index0Param1Write=false
    Module.Index0Param2Write=false
    Module.Index0Param3Write=false
   )
  ]
  rendererClasses [&NiagaraSpriteRendererProperties]
  renderers [{rendererClass=&NiagaraSpriteRendererProperties}]
 }
 Smoke_Combined {
  emitterUpdateScript [
   &EmitterState()
   &SpawnBurst_Instantaneous(
    Spawn Count=&RandomRangeInt(Minimum=2, Maximum=3)
    Use Spawn Probability=false
    UseLoopCountLimit=false
   )
   &SpawnBurst_Instantaneous(Spawn Count=5, Spawn Group=1, Use Spawn Probability=false, UseLoopCountLimit=false) {module=SpawnBurst_Instantaneous001}
  ]
  particleSpawnScript [
   /SetVariables(Particles.VisibilityTag=&SpawnGroup()) {isSetParameters=true}
   &InitializeParticle(
    Color Maximum=&MaskLinearColorBySpawnGroup(Masked Linear Color=<rgba 0.0495 0.04 0.0729 0.2>, Passthrough Linear Color=<rgba 0.015 0.015 0.015 1>)
    Color Minimum=&MaskLinearColorBySpawnGroup(Masked Linear Color=<rgba 0.0431 0.0431 0.0778 0.15>, Passthrough Linear Color=<rgba 0.015 0.015 0.015 1>)
    Color Mode=::NewEnumerator2
    Lifetime Max=&MaskFloatBySpawnGroup(Masked Float=0.4, Passthrough Float=0.75)
    Lifetime Min:NiagaraFloat=&MaskFloatBySpawnGroup(Masked Float=0.3, Passthrough Float=0.5)
    Lifetime Mode=::NewEnumerator1
    Sprite Rotation Mode=::NewEnumerator2
    Sprite Size Max=&MaskVector2dBySpawnGroup(Masked Vector2D=<130 130>, Passthrough Vector2D=<50 50>, Spawn Group Mask=0)
    Sprite Size Min=&MaskVector2dBySpawnGroup(Masked Vector2D=<100 100>, Passthrough Vector2D=<30 30>, Spawn Group Mask=0)
    Sprite Size Mode=::NewEnumerator2
    Sprite UV Mode=::NewEnumerator3
    UsePositionOffset=false
    Write Material Random=true
   )
   &CameraOffset(Camera Offset Amount=&MaskFloatBySpawnGroup(
     Masked Float=&RandomRangeFloat(Minimum=40, Maximum=50, Evaluation Type=::NewEnumerator1)
     Passthrough Float=5
    ))
   &ShapeLocation(
    Hemisphere Distribution=<1 0.5>
    MaskSpawnGroup=false
    Override Seed=false
    Rotation Angle=90
    Rotation Mode=::NewEnumerator1
    Sphere Radius=&MaskFloatBySpawnGroup(
     Masked Float=&MakeCustomFloatFromBool(Bool=$User.bIsFrontEnd, True Float=5, False Float=10)
     Passthrough Float=&Multiply_Float(A=10, B=$System.GlobalScale)
    )
    UseOwnerScale=true
   )
   &AddVelocity(
    Module.Use Falloff Distance=false
    Override Seed=false
    Velocity Mode=::NewEnumerator1
    Velocity Speed=&MaskFloatBySpawnGroup(
     Masked Float=0
     Passthrough Float=&Multiply_Float(
      A=&RandomRangeFloat(Minimum=25, Maximum=50)
      B=&FN_FrontEndFloatModifier(Gameplay=1, Front End=0)
     )
    )
   ) {module=AddVelocity001}
   &AddVelocity(
    Velocity=&MaskVectorBySpawnGroup(
     Masked Vector=&Multiply_VectorByFloat(Vector=$Particles.ShapeLocation.ShapeVector, Float=10)
     Passthrough Vector=$System.DirectionVector
    )
    Velocity Speed Scale=&MaskFloatBySpawnGroup(
     Masked Float=1
     Passthrough Float=&Multiply_Float(
      A=&RandomRangeFloat(Minimum=50, Maximum=125)
      B=&FN_FrontEndFloatModifier(Gameplay=1, Front End=0)
     )
    )
   )
  ]
  particleUpdateScript [
   &ParticleState()
   &DynamicMaterialParameters(
    Index 0 Param 1=&Add_Float(
     A=&RandomRangeFloat(Minimum=0, Maximum=1)
     B=&FloatFromCurve(
      FloatCurve=DataInterface() {propertyValues {Curve {keys [
          <at 0 0 RCIM_Cubic>
          {arriveTangent=1.5947, interpMode=RCIM_Cubic, leaveTangent=1.5947, tangentMode=RCTM_User, time=0.2557, value=0.5257}
          <at 1 1 RCIM_Cubic>
         ]}}}
      CurveIndex=$Particles.Age
      Scale Curve=0.25
     )
    )
    Module.Index0Float0Write=true
    Module.Index0Float1Write=true
    Module.Index0Float2Write=true
    Module.Index0Float3Write=true
   ) {module=DynamicMaterialParameters001}
   &SubUVAnimation(
    End Frame Range Override=30
    Output Alternate Renderer Binding=true
    "Sprite Renderer "=DataInterface() {propertyValues {SpriteRenderer=&self:Smoke_Combined_0.NiagaraSpriteRendererProperties_2}}
    Start Frame Range Override=&RandomRangeInt(Minimum=0, Maximum=5)
    UseEndFrame=true
    UseStartFrame=true
   )
   &ScaleColor(
    Scale Alpha=&MaskFloatBySpawnGroup(
     Masked Float=&FloatFromCurve(FloatCurve=DataInterface() {propertyValues {Curve {keys [
          {arriveTangent=2.4358, interpMode=RCIM_Cubic, leaveTangent=2.4358, tangentMode=RCTM_User, time=0, value=0}
          <at 0.25 0.5 RCIM_Cubic>
          <at 1 0 RCIM_Cubic>
         ]}}})
     Passthrough Float=&Multiply_Float(
      A=&NormalizeDistanceRange(
       Start Position:NiagaraPosition=$Particles.Initial.Position
       End Position:NiagaraPosition=$Particles.Position
       Distance:NiagaraFloat=100
       Invert Normalized Range:NiagaraBool=true
      )
      B=&FloatFromCurve(
       FloatCurve=DataInterface() {propertyValues {Curve {keys [
           {arriveTangent=7.7616, interpMode=RCIM_Cubic, leaveTangent=7.7616, tangentMode=RCTM_User, time=0, value=0.5}
           {arriveTangent=-0.0888, interpMode=RCIM_Cubic, leaveTangent=-0.0888, tangentMode=RCTM_User, time=0.1, value=1}
           {arriveTangent=-3.9148, interpMode=RCIM_Cubic, leaveTangent=-3.9148, tangentMode=RCTM_User, time=1, value=0}
          ]}}}
       Scale Curve=&MakeCustomFloatFromBool(Bool=$User.bIsFrontEnd, True Float=0.5, False Float=1)
      )
     )
    )
    Scale RGB=&VectorFromCurve(VectorCurve=DataInterface() {propertyValues {XCurve {keys [<at 0 1>,<at 1 1>]}, YCurve {keys [<at 0 1>,<at 1 1>]}, ZCurve {keys [<at 0 1>,<at 1 1>]}}})
    ScaleA=true
    ScaleRGB=true
   )
   &ScaleSpriteSize(
    Scale Sprite Size Mode=::NewEnumerator0
    Uniform Scale Factor=&MaskFloatBySpawnGroup(
     Masked Float=&FloatFromCurve(FloatCurve=DataInterface() {propertyValues {Curve {keys [
          {arriveTangent=1.7761, interpMode=RCIM_Cubic, leaveTangent=1.7761, tangentMode=RCTM_User, time=0, value=0}
          {interpMode=RCIM_Cubic, tangentMode=RCTM_User, time=1, value=1}
         ]}}})
     Passthrough Float=&FloatFromCurve(
      FloatCurve=DataInterface() {propertyValues {Curve {
         defaultValue=0.8
         keys [
          {arriveTangent=1.6474, interpMode=RCIM_Cubic, leaveTangent=1.6474, tangentMode=RCTM_User, time=0, value=0.5}
          {arriveTangent=0.3458, interpMode=RCIM_Cubic, leaveTangent=0.3458, tangentMode=RCTM_User, time=1, value=1}
         ]
        }}}
      Scale Curve=&Multiply_Float(A=1.3, B=$System.GlobalScale)
     )
    )
   )
   &SpriteRotationRate(Rotation Rate=&MaskFloatBySpawnGroup(Masked Float=0, Passthrough Float=&RandomRangeFloat(Minimum=-50, Maximum=50)))
   &GravityForce(Gravity=&MaskVectorBySpawnGroup(Masked Vector=<0 0 0>, Passthrough Vector=<0 0 80>))
   &Drag(
    Drag=&MaskFloatBySpawnGroup(Masked Float=0, Passthrough Float=&RandomRangeFloat(Minimum=2, Maximum=3))
    Use Linear Drag=true
    Use Rotational Drag=true
   )
   &SolveForcesAndVelocity(Clamp Velocity=false, Limit Acceleration=false)
  ]
  rendererClasses [&NiagaraSpriteRendererProperties]
  renderers [
   {rendererClass=&NiagaraSpriteRendererProperties}
   {rendererClass=&NiagaraSpriteRendererProperties, rendererIndex=1}
   {rendererClass=&NiagaraSpriteRendererProperties, rendererIndex=2}
  ]
 }
 Sparks_CPU {
  emitterUpdateScript [
   &EmitterState(
    Enable Distance Culling=true
    Max Distance Response=::KillImmediately
    Module.bApplySpawnCountScaleByDistance=true
    Module.bUseMaxDistance=true
    Module.bUseMinDistance=false
    Scalability Mode=::NewEnumerator1
    Scale Spawn Count=true
    Spawn Count Scale By Distance=DataInterface() {propertyValues {Curve {keys [<at 0.3676 0.9657>,<at 0.9962 0.4982>]}}}
   )
   &SpawnBurst_Instantaneous(Spawn Count=29, Use Spawn Probability=true, UseLoopCountLimit=false)
   &SpawnBurst_Instantaneous(
    Spawn Count=&RandomRangeInt(Minimum=15, Maximum=20)
    Spawn Group=1
    Spawn Probability=&FN_FrontEndFloatModifier(Gameplay=0, Front End=1)
    Use Spawn Probability=true
    UseLoopCountLimit=false
   ) {enabled=false, module=SpawnBurst_Instantaneous001}
  ]
  particleSpawnScript [
   &InitializeParticle(
    Color=<rgba 0.237 0.1043 0.1043 1>
    Lifetime=&FloatFromCurve(
     FloatCurve=DataInterface() {propertyValues {Curve {keys [
         {arriveTangent=-5.4544, interpMode=RCIM_Cubic, leaveTangent=-5.4544, tangentMode=RCTM_User, time=0, value=0.75}
         {arriveTangent=-0.65, interpMode=RCIM_Cubic, leaveTangent=-0.65, time=0.1, value=0.4}
         <at 1 0.1>
        ]}}}
     CurveIndex=&RandomRangeFloat(Minimum=0, Maximum=1)
    )
    Sprite Size Mode=::NewEnumerator3
    Uniform Sprite Size=&Multiply_Float(A=$System.GlobalScale, B=&RandomRangeFloat(Minimum=8, Maximum=14))
    UsePositionOffset=false
    Write Material Random=true
   )
   &ShapeLocation(
    MaskSpawnGroup=false
    Override Seed=false
    Rotation Axis=$System.DirectionVector
    Rotation Mode=::NewEnumerator1
    Sphere Radius=5
    UseOwnerScale=true
   )
   &CameraOffset(Camera Offset Amount=10)
   &AddVelocityFromPoint(
    Module.Use Falloff Distance=false
    Velocity Strength=&Multiply_Float(
     A=&RandomRangeFloat(Minimum=150, Maximum=200)
     B=&FloatFromCurve(
      FloatCurve=DataInterface() {propertyValues {Curve {keys [<at 0.7 1>,<at 0.85 3>,<at 1 3>]}}}
      CurveIndex=&RandomRangeFloat(Minimum=0, Maximum=1)
     )
    )
   )
   &AddVelocity(
    Velocity=$System.DirectionVector
    Velocity Speed Scale=&Multiply_Float(A=&RandomRangeFloat(Minimum=200, Maximum=500), B=$System.GlobalScale)
   )
   &AddVelocity(Module.Use Falloff Distance=false, Override Seed=false, Velocity Mode=::NewEnumerator1) {enabled=false, module=AddVelocity001}
   &DynamicMaterialParameters(
    Index 0 Param 1=3
    Index 0 Param 2=20
    Index 0 Param 3=1
    Module.Index0Float0Write=true
    Module.Index0Float1Write=true
    Module.Index0Float2Write=true
    Module.Index0Float3Write=true
   )
  ]
  particleUpdateScript [
   &ParticleState()
   &AccelerationForce(Acceleration=&RandomRangeVector(Minimum=<0 0 -2000>, Maximum=<0 0 -1000>))
   &Drag(Drag=&UniformRangedFloat(Minimum=2, Maximum=6), Use Linear Drag=true, Use Rotational Drag=true)
   &GravityForce(Gravity=&MaskVectorBySpawnGroup(Masked Vector=<0 0 0>, Passthrough Vector=<0 0 -666>)) {enabled=false}
   &SolveForcesAndVelocity(Clamp Velocity=false, Limit Acceleration=false)
   &ScaleSpriteSizeBySpeed(Max Scale Factor=<0.1 1>, Min Scale Factor=<0.25 0.5>, Velocity Threshold=666)
   &ScaleSpriteSize(Uniform Curve Sprite Scale=DataInterface() {propertyValues {Curve {keys [
        {
         arriveTangent=6.3093
         interpMode=RCIM_Cubic
         leaveTangent=6.3093
         leaveTangentWeight=0.9891
         tangentMode=RCTM_User
         time=0.0028
         value=0.0699
        }
        {
         arriveTangent=0.1864
         arriveTangentWeight=0.2242
         interpMode=RCIM_Cubic
         leaveTangent=-4.7115
         leaveTangentWeight=0.7932
         tangentMode=RCTM_Break
         time=0.65
         value=1
        }
        {arriveTangentWeight=0.5326, interpMode=RCIM_Cubic, tangentMode=RCTM_User, time=1, value=0}
       ]}}})
   &ScaleColor(
    Scale Alpha=&FloatFromCurve(
     FloatCurve=DataInterface() {propertyValues {Curve {keys [
         {interpMode=RCIM_Cubic, tangentMode=RCTM_User, time=0, value=1}
         {arriveTangent=-0.3845, interpMode=RCIM_Cubic, leaveTangent=-0.3845, tangentMode=RCTM_User, time=0.5, value=0.9}
         {arriveTangent=-0.2828, interpMode=RCIM_Cubic, leaveTangent=-0.2828, tangentMode=RCTM_User, time=1, value=0}
        ]}}}
     Scale Curve=&RandomRangeFloat(Minimum=0.25, Maximum=5)
    )
    Scale RGB=&Multiply_VectorByFloat(
     Vector=&MakeVectorFromLinearColorRGB(Color=&ColorFromCurve(ColorCurve=DataInterface() {propertyValues {
         AlphaCurve {keys [<at 0.333 1>,<at 1 0>]}
         BlueCurve {keys [<at 0 2.249>,<at 0.2991 0>]}
         GreenCurve {keys [<at 0 2.3079>,<at 0.2991 0.0016>]}
         RedCurve {keys [<at 0 3>,<at 0.2991 0.0335>]}
        }}))
     Float=2
    )
    ScaleA=true
    ScaleRGB=true
   )
  ]
  rendererClasses [&NiagaraSpriteRendererProperties]
  renderers [{rendererClass=&NiagaraSpriteRendererProperties}]
 }
 Sparks_Gpu {
  simTarget=GPUComputeSim
  emitterUpdateScript [
   &EmitterState()
   &SpawnBurst_Instantaneous(Spawn Count=&UniformRangedInt(Minimum=6, Maximum=15), Use Spawn Probability=false)
   &SpawnRate(
    Spawn Probability=&FloatFromCurve(
     FloatCurve=DataInterface() {propertyValues {Curve {keys [{arriveTangent=-12.2872, interpMode=RCIM_Cubic, leaveTangent=-12.2872, tangentMode=RCTM_User, time=0, value=1}, <at 0.1 0>]}}}
     CurveIndex=$System.Age
    )
    SpawnRate=100
    Use Spawn Probability=true
   )
  ]
  particleSpawnScript [
   &InitializeParticle(
    Color=<rgba 0.237 0.1043 0.1043 1>
    Lifetime=&FN_FrontEndFloatModifier(
     Gameplay=&RandomRangeFloat(Minimum=0.25, Maximum=1.1)
     Front End=&RandomRangeFloat(Minimum=0.25, Maximum=0.5)
    )
    Sprite Size Mode=::NewEnumerator3
    Uniform Sprite Size=&Multiply_Float(
     A=$System.GlobalScale
     B=&FloatFromCurve(
      FloatCurve=DataInterface() {propertyValues {Curve {keys [<at 0 13>,<at 0.2 7>,<at 1 2.5>]}}}
      CurveIndex=&RandomRangeFloat(Minimum=0, Maximum=1)
     )
    )
    UsePositionOffset=false
    Write Material Random=true
   )
   &ShapeLocation(
    MaskSpawnGroup=false
    Override Seed=false
    Rotation Angle=90
    Rotation Mode=::NewEnumerator1
    Sphere Radius=10
    Sphere Surface Distribution=1
    UseOwnerScale=true
   )
   &AddVelocity(
    Module.Use Falloff Distance=false
    Override Seed=false
    Velocity Mode=::NewEnumerator1
    Velocity Speed=&Multiply_Float(A=$System.GlobalScale, B=&RandomRangeFloat(Minimum=100, Maximum=300))
   )
   &CurlNoiseForce(
    Noise Frequency=75
    Noise Strength=&RandomRangeFloat(Minimum=100, Maximum=250)
    Pan Noise=true
    Pan Noise Field=<2 2 0>
   )
   &ApplyInitialForces()
   &DynamicMaterialParameters(
    Index 0 Param 1=3
    Index 0 Param 2=20
    Index 0 Param 3=1
    Module.Param0WriteEnabled=true
    Module.Param1WriteEnabled=true
    Module.Param2WriteEnabled=true
    Module.Param3WriteEnabled=true
   )
   &AddVelocity(
    Module.Use Falloff Distance=false
    Override Seed=false
    Velocity Mode=::NewEnumerator1
    Velocity Speed=&Multiply_Float(A=$System.GlobalScale, B=&RandomRangeFloat(Minimum=300, Maximum=400))
   ) {enabled=false, module=AddVelocity001}
  ]
  particleUpdateScript [
   &ParticleState()
   &GravityForce(Gravity=&RandomRangeVector(Minimum=<0 0 -700>, Maximum=<0 0 -980>))
   &Drag(Drag=&UniformRangedFloat(Minimum=1.5, Maximum=3), Use Linear Drag=true, Use Rotational Drag=false)
   &Collision(
    Age Colliding Particles=false
    Kill Offscreen Particles=true
    KillOnCollision=false
    Module.Control Rest State Via Surface Normal=false
    Module.EnableMaxCollisionCount=false
    Module.Randomize Collision Normal Vector=false
   )
   &SolveForcesAndVelocity(Clamp Velocity=false, Limit Acceleration=false)
   &ScaleSpriteSize(Scale Factor=&Vector2DFromCurve(
     Vector2Curve=DataInterface() {propertyValues {
       XCurve {keys [
         {arriveTangent=5.9728, interpMode=RCIM_Cubic, leaveTangent=5.9728, tangentMode=RCTM_User, time=0, value=0.1}
         {arriveTangent=-0.1, interpMode=RCIM_Cubic, leaveTangent=-0.1, tangentMode=RCTM_User, time=0.18, value=0.8}
         {interpMode=RCIM_Cubic, tangentMode=RCTM_User, time=1, value=0}
        ]}
       YCurve {keys [
         {arriveTangent=6.2715, interpMode=RCIM_Cubic, leaveTangent=6.2715, tangentMode=RCTM_User, time=0, value=0.1}
         {arriveTangent=-0.1001, interpMode=RCIM_Cubic, leaveTangent=-0.1001, tangentMode=RCTM_User, time=0.18, value=0.8}
         {interpMode=RCIM_Cubic, tangentMode=RCTM_User, time=0.9988, value=0}
        ]}
      }}
     ScaleCurve=<1 1>
    ))
   &ScaleSpriteSizeBySpeed(Max Scale Factor=<0.5 3>, Min Scale Factor=<0.5 1>, Velocity Threshold=666)
   &ScaleColor(
    Linear Color Curve=DataInterface() {propertyValues {
      AlphaCurve {keys [<at 0.333 1>,<at 1 0>]}
      BlueCurve {keys [<at 0 3.0115>,<at 0.208 0>]}
      GreenCurve {keys [<at 0 3.0903>,<at 0.208 0.0016>]}
      RedCurve {keys [<at 0 4.0171>,<at 0.208 0.0335>]}
     }}
    Scale Mode=::NewEnumerator2
   )
  ]
  rendererClasses [&NiagaraSpriteRendererProperties]
  renderers [{rendererClass=&NiagaraSpriteRendererProperties}]
 }
}
summary {
 systemName=NS_Pickaxe_Bass_PolkaSkate_Impact
 emitters [
  {emitterName=Sparks_Gpu, simTarget=GPUComputeSim, rendererClasses [&NiagaraSpriteRendererProperties]}
  {
   emitterName=FN_PhysMatDebris
   rendererClasses [&NiagaraSpriteRendererProperties, &NiagaraMeshRendererProperties]
  }
  {
   emitterName=Core_Impact_Burst
   rendererClasses [
    &NiagaraSpriteRendererProperties
    &NiagaraLightRendererProperties
    &NiagaraMeshRendererProperties
   ]
  }
  {emitterName=Petals, rendererClasses [&NiagaraSpriteRendererProperties]}
  {emitterName=Sparks_CPU, rendererClasses [&NiagaraSpriteRendererProperties]}
  {emitterName=Smoke_Combined, rendererClasses [&NiagaraSpriteRendererProperties]}
  {emitterName=Directional_Elements, rendererClasses [&NiagaraSpriteRendererProperties]}
 ]
 userVariables [
  User.bIsFrontEnd {type=NiagaraBool}
  User.bIsFrontEndPreview {type=NiagaraBool}
  User.PhysicalSurface {type=NiagaraInt32}
  User.SwingSectionIndex {type=NiagaraInt32}
 ]
}
