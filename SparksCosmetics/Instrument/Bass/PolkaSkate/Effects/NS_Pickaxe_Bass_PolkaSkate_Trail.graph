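# Niagara trail effect for the PolkaSkate bass pickaxe cosmetic, sampling the user-supplied skeletal mesh.
# Two anim-trail ribbons draw a dark red streak, and a curl-noise smoke ribbon spawned from the mesh widens and fades.
# Petal sprites spawn at mesh bones or sockets, inherit velocity and flutter down under gravity, curl noise and aerodynamic drag.
# A system time-of-day color tints the smoke and one trail, and a front-end flag halves smoke opacity.
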NiagaraSystem /SparksCosmetics/Instrument/Bass/PolkaSkate/Effects/NS_Pickaxe_Bass_PolkaSkate_Trail
dependencies {
 usedDataInterfaces [
  &NiagaraDataInterfaceCurve
  &NiagaraDataInterfaceColorCurve
  &NiagaraDataInterfaceArrayColor
  &NiagaraDataInterfaceArrayInt32
 ]
 usedDynamicInputs []
 usedModules [
  &self:SystemSpawnScript.SystemScriptSource.SystemScriptGraph.NiagaraNodeAssignment_0.SetVariables_#1
  &self:SystemSpawnScript.SystemScriptSource.SystemScriptGraph.NiagaraNodeAssignment_1.SetVariables_#2
  &SystemState
  &BlendSpaceHelper
 ]
}
emitters {
 Petals_Falling {
  emitterUpdateScript [
   &EmitterState(Life Cycle Mode=::NewEnumerator1, Loop Behavior=::NewEnumerator1, Loop Duration=0.25, UseLoopDelay=false)
   &SpawnRate(
    SpawnRate=&RandomRangeFloat(Minimum=0, Maximum=40, Evaluation Type=::NewEnumerator1)
    Use Spawn Probability=false
   ) {module=SpawnRate001}
   &SpawnRate(
    SpawnRate=&FloatFromCurve(
     FloatCurve=DataInterface() {propertyValues {Curve {keys [
         {interpMode=RCIM_Cubic, leaveTangentWeight=0.5055, tangentMode=RCTM_User, time=0, value=1}
         {arriveTangent=-0.0083, tangentMode=RCTM_User, time=0.8, value=0}
        ]}}}
     CurveIndex=$Emitter.NormalizedLoopAge
     Scale Curve=30
    )
    Use Spawn Probability=false
   ) {enabled=false}
   /SetVariables(
    Emitter.WindOscillateTime:NiagaraFloat=&Divide_Float(A=$Engine.Time, B=44)
    Emitter.WindRotateTime:NiagaraFloat=&Divide_Float(A=$Engine.Time, B=66)
   ) {enabled=false, isSetParameters=true}
   /SetVariables(
    Emitter.WindIntensity:NiagaraFloat=&Multiply_Float(A=&Sine(Normalized Angle=$Emitter.WindOscillateTime, Period=1, Scale=1, Bias=0), B=0)
    Emitter.WindVector:Vector3f=&MakeVector(
     X=&Add_Float(
      A=&Sine(Normalized Angle=$Emitter.WindRotateTime, Period=1, Scale=1, Bias=0)
      B=&Cosine(Normalized Angle=$Emitter.WindRotateTime, Period=1, Scale=1, Bias=0)
     )
     Y=&Subtract_Float(
      A=&Cosine(Normalized Angle=$Emitter.WindRotateTime, Period=1, Scale=1, Bias=0)
      B=&Sine(Normalized Angle=$Emitter.WindRotateTime, Period=1, Scale=1, Bias=0)
     )
     Z=0
    )
   ) {enabled=false, isSetParameters=true}
  ]
  particleSpawnScript [
   &InitializeParticle(
    Color=&SelectLinearColorFromArray(
     Array Sampling Mode=::NewEnumerator0
     Direct Array Index=?
     Color Selection Array=DataInterface() {propertyValues {ColorData [<rgba 0.015 0 0.0002 1>, <rgba 0.02 0.0019 0 1>]}}
     Random Seed=0
    )
    Lifetime Max=0.55
    Lifetime Min=0.25
    Lifetime Mode=::NewEnumerator1
    Mass=&Multiply_Float(
     A=&RandomRangeFloat(Minimum=2, Maximum=4)
     B=&FloatFromCurve(
      FloatCurve=DataInterface() {propertyValues {Curve {keys [<at 0 0.5 RCIM_Cubic>, {arriveTangent=1.5048, interpMode=RCIM_Cubic, leaveTangent=1.5048, tangentMode=RCTM_User, time=1, value=1}]}}}
      CurveIndex=&ReturnNormalizedExecIndex()
      Scale Curve=1.5
     )
    )
    Mass Mode=::NewEnumerator1
    Sprite Rotation Mode=::NewEnumerator2
    Sprite Size Max=<8 6>
    Sprite Size Min=<4 2>
    Sprite Size Mode=::NewEnumerator2
    UsePositionOffset=false
    Write Material Random=true
   )
   &ShapeLocation(
    MaskSpawnGroup=false
    Offset=<0 0 50>
    Offset Coordinate Space=::Local
    Offset Mode=::NewEnumerator0
    Override Seed=false
    Sphere Radius=2.5
    UseOwnerScale=true
   ) {enabled=false}
   &SkeletalMeshLocation(
    Bone / Socket Index=&SelectIntFromArray(
     Array Sampling Mode=::NewEnumerator1
     Direct Array Index=&RandomRangeInt(Minimum=0, Maximum=1)
     Int Selection Array=DataInterface() {propertyValues {IntData [0,1]}}
     Random Seed=0
    )
    Bone Sampling Mode=::NewEnumerator4
    "Inherit Velocity (Scale)":NiagaraFloat=&RandomRangeFloat(Minimum=0.7, Maximum=2)
    Inherited Velocity Speed Limit=100
    MaskSpawnGroup=false
    Orientation Sampling=::NewEnumerator1
    Sampled Position Offset=&Multiply_VectorByFloat(Vector=&RandomRangeVector(Minimum=<-1 -1 -1>, Maximum=<1 1 1>), Float=7)
    Skeletal Mesh=$User.SkeletalMesh
    Velocity Sampling=::NewEnumerator0
    VelocitySpeedLimit=true
   )
   &InitialMeshOrientation(Use Rotation Vector=false)
   &InheritVelocity(
    Inherited Velocity Amount Scale=&Multiply_VectorByFloat(Vector=<1 1 1>, Float=100)
    Speed Limit=true
    UseSpeedThreshold=false
   )
  ]
  particleUpdateScript [
   &ParticleState()
   &BlendSpace(
    Mesh Orientation Mode=1
    "Position Blend (World <-> Local)":NiagaraFloat=&FloatFromCurve(
     FloatCurve=DataInterface() {propertyValues {Curve {
        defaultValue=0
        keys [{interpMode=RCIM_Cubic, tangentMode=RCTM_User, time=0, value=0.5}, {interpMode=RCIM_Cubic, tangentMode=RCTM_User, time=0.5, value=0}]
       }}}
     Translation Contribution:NiagaraFloat=1
     Rotation Contribution:NiagaraFloat=1
    )
    Position Coordinate Space=1
    Velocity Coordinate Space=1
    Velocity Mode=1
   ) {enabled=false, module=BlendSpace001}
   &ScaleColor(
    Scale Alpha=&FloatFromCurve(FloatCurve=DataInterface() {propertyValues {Curve {keys [<at -0.0046 1.0113 RCIM_Cubic>,<at 1.0011 0.002 RCIM_Cubic>]}}})
    ScaleA=true
    ScaleRGB=false
   )
   &ScaleSpriteSize(
    Scale Factor=&Multiply_Vector2(
     A=&MakeVector2D(
      X=&Multiply_Float(
       A=1
       B=&Sine(
        Normalized Angle=$Emitter.Age
        Period=&Multiply_Float(A=3, B=&Add_Float(A=$Particles.Initial.MaterialRandom, B=0.6))
        Scale=1
        Bias=0
       )
      )
      Y=1
     )
     B=&Vector2DFromFloat(Value=&FloatFromCurve(
       FloatCurve=DataInterface() {propertyValues {Curve {keys [
           {arriveTangent=8.2412, interpMode=RCIM_Cubic, leaveTangent=8.2412, tangentMode=RCTM_User, time=0, value=0.5}
           {arriveTangent=0.0073, interpMode=RCIM_Cubic, leaveTangent=0.0073, tangentMode=RCTM_User, time=0.1, value=1}
           {arriveTangent=-1.2169, interpMode=RCIM_Cubic, leaveTangent=-1.2169, tangentMode=RCTM_User, time=0.5538, value=0.7921}
           <at 1 0.25>
          ]}}}
       Scale Curve=&Multiply_Float(A=&RandomRangeFloat(Minimum=1.75, Maximum=2.25), B=$System.GlobalScale)
      ))
    )
    Scale Sprite Size Mode=::NewEnumerator2
   ) {module=ScaleSpriteSize001}
   &SpriteRotationRate(Rotation Rate=&FloatFromCurve(
     FloatCurve=DataInterface() {propertyValues {Curve {keys [
         {interpMode=RCIM_Cubic, tangentMode=RCTM_User, time=0, value=1}
         {interpMode=RCIM_Cubic, leaveTangent=-4.1325, tangentMode=RCTM_Break, time=0.5, value=1}
         {interpMode=RCIM_Cubic, tangentMode=RCTM_User, time=1, value=0}
        ]}}}
     Scale Curve=&RandomRangeFloat(Minimum=180, Maximum=1000)
    )) {enabled=false}
   &AccelerationForce(Acceleration=&RandomRangeVector(Minimum=<0 0 99>, Maximum=<0 0 199>))
   &CurlNoiseForce(
    Bias Noise=false
    Noise Strength=&FloatFromCurve(
     FloatCurve=DataInterface() {propertyValues {Curve {
        defaultValue=1
        keys [
         {arriveTangent=2.0766, interpMode=RCIM_Cubic, leaveTangent=2.0766, tangentMode=RCTM_User, time=0.2, value=0}
         {interpMode=RCIM_Cubic, tangentMode=RCTM_User, time=1, value=1}
        ]
       }}}
     Scale Curve=150
    )
    Pan Noise=false
   )
   &Drag(Drag=3, Use Linear Drag=true, Use Rotational Drag=true)
   &GravityForce(Gravity=&Multiply_VectorByFloat(Vector=&RandomRangeVector(Minimum=<0 0 -720>, Maximum=<0 0 -980>), Float=1))
   &PointForce(
    Force Origin=&AddVectorToPosition(Position=/Niagara/DynamicInputs/Helpers/SimulationPosition(), Vector=<0 0 50>)
    Force Strength=1000
    Module.Use Falloff=true
    Module.UseLinearForce=false
   ) {enabled=false}
   &WindForce(Wind Speed=$Emitter.WindVector, Wind Speed Scale=$Emitter.WindIntensity) {enabled=false}
   &AerodynamicDrag(
    Aerodynamic Drag:NiagaraFloat=&FloatFromCurve(
     FloatCurve=DataInterface() {propertyValues {Curve {
        defaultValue=1
        keys [
         {arriveTangent=2.3105, interpMode=RCIM_Cubic, leaveTangent=2.3105, tangentMode=RCTM_User, time=0, value=0}
         {interpMode=RCIM_Cubic, tangentMode=RCTM_User, time=1, value=1}
        ]
       }}}
     Scale Curve=&RandomRangeFloat(Minimum=1, Maximum=2)
    )
    Drag Shape Curve=DataInterface() {propertyValues {Curve {keys [
        {interpMode=RCIM_Cubic, tangentMode=RCTM_User, time=0, value=0.25}
        {arriveTangent=2.0799, interpMode=RCIM_Cubic, leaveTangent=2.0799, tangentMode=RCTM_User, time=1, value=1}
       ]}}}
    Lift Shape Curve=DataInterface() {propertyValues {Curve {keys [
        {interpMode=RCIM_Cubic, tangentMode=RCTM_User, time=0, value=0}
        {interpMode=RCIM_Cubic, tangentMode=RCTM_User, time=0.2, value=1}
        {arriveTangent=-0.375, interpMode=RCIM_Cubic, leaveTangent=-0.375, time=0.5, value=0.35}
        {tangentMode=RCTM_User, time=1, value=0.7}
       ]}}}
    Pivot Offset Minimum Length:Vector3f=<0.15 0 1>
    Rotation Strength=2.5
    Surface Normal Axis:Vector3f=&RandomRangeVector(Minimum=<-0.5 -0.5 1>, Maximum=<0.5 0.5 1>)
    Use Lift Smoothing=true
    Use Linear Drag=true
    Use Rotational Drag=true
   )
   &SolveForcesAndVelocity(Clamp Velocity=false, Limit Acceleration=false)
   &AlignSpriteToMeshOrientation(SetAlignment=true, SetFacing=true)
   &DynamicMaterialParameters(
    Index 0 Param 1=&RandomRangeFloat(Minimum=0, Maximum=0.5)
    Module.Index0Param0Write=true
    Module.Index0Param1Write=false
    Module.Index0Param2Write=false
    Module.Index0Param3Write=false
   )
  ]
  rendererClasses [&NiagaraSpriteRendererProperties]
  renderers [{rendererClass=&NiagaraSpriteRendererProperties}]
 }
 Ribbon_SmokeTrail {
  emitterUpdateScript [
   &EmitterState(Life Cycle Mode=::NewEnumerator1, Loop Behavior=::NewEnumerator1, Loop Delay=0.005, Loop Duration=1, UseLoopDelay=true)
   &SpawnRate(SpawnRate=45, Use Spawn Probability=false)
  ]
  particleSpawnScript [
   &InitializeRibbon(
    Color=<rgba 0 0 0 1>
    Lifetime=0.3
    Module.Write Color=true
    Module.Write Lifetime=true
    Module.Write Mass=false
    Module.Write Material Random=true
    Module.Write Position=false
    Module.Write RibbonTwist=false
    Module.Write RibbonWidth=true
    Ribbon Width=27
   )
   &SkeletalMeshLocation(
    Inherit Velocity=false
    MaskSpawnGroup=false
    Mesh Sampling Type=::NewEnumerator5
    Sampled Position Offset=<0 0 10>
    Sampled Position Offset Coordinate Space:ENiagaraMeshTransforms=::NewEnumerator3
    Skeletal Mesh=$User.SkeletalMesh
    Socket Sampling Mode=::NewEnumerator5
   )
   &AddVelocity(Velocity=&RandomVector(Vector Scale=&FloatFromCurve(FloatCurve=DataInterface() {propertyValues {Curve {keys [<at 0 0>,<at 1 1>]}}}, Scale Curve=1000))) {enabled=false}
  ]
  particleUpdateScript [
   &ParticleState()
   &CurlNoiseForce(
    Bias Noise=false
    Noise Frequency=300
    Noise Strength=&FloatFromCurve(
     FloatCurve=DataInterface() {propertyValues {Curve {
        defaultValue=1
        keys [
         {arriveTangent=4.1492, interpMode=RCIM_Cubic, leaveTangent=4.1492, tangentMode=RCTM_User, time=0.5, value=0}
         {interpMode=RCIM_Cubic, tangentMode=RCTM_User, time=0.9, value=1}
        ]
       }}}
     Scale Curve=5000
    )
    Pan Noise=false
   )
   &SolveForcesAndVelocity(Clamp Velocity=false, Limit Acceleration=false)
   &Color(
    Color=&ColorFromCurve(ColorCurve=DataInterface() {propertyValues {
       AlphaCurve {keys [<at 0.7149 2 RCIM_Cubic>,<at 0.9985 0>]}
       BlueCurve {keys [<at 0.0057 0.0189>,<at 0.2591 0>,<at 1 0.0156>]}
       GreenCurve {keys [<at 0.0057 0>,<at 0.2591 0>,<at 1 0.0026>]}
       RedCurve {keys [<at 0.0057 0.0039>,<at 0.2591 0>,<at 1 0>]}
      }})
    Scale Alpha=0.1
   )
   &ScaleColor(
    Scale Alpha=&Multiply_Float(
     A=&FloatFromCurve(
      FloatCurve=DataInterface() {propertyValues {Curve {keys [
          {interpMode=RCIM_Cubic, leaveTangent=-1.9403, leaveTangentWeight=0.4326, tangentMode=RCTM_Break, time=0.4, value=1}
          {arriveTangent=-1.1087, arriveTangentWeight=0.6923, interpMode=RCIM_Cubic, leaveTangent=-1.1087, tangentMode=RCTM_User, time=1, value=0}
         ]}}}
      Scale Curve=&FloatFromCurve(
       FloatCurve=DataInterface() {propertyValues {Curve {keys [<at 0 1>,<at 0.497 1 RCIM_Cubic>,<at 1 1>]}}}
       CurveIndex=Linked() {linkedVariable {name=NPC.FortniteNPC.TimeOfDay_0-1, type=NiagaraFloat}}
      )
     )
     B=&MakeCustomFloatFromBool(Bool=$User.bIsFrontEnd, True Float=0.5, False Float=1)
    )
    Scale RGB=&MakeVectorFromLinearColorRGB(Color=$System.TOD Color)
    ScaleA=true
    ScaleRGB=true
   ) {module=ScaleColor002}
   &ScaleRibbonWidth(Ribbon Width Scale=&FloatFromCurve(FloatCurve=DataInterface() {propertyValues {Curve {keys [
         {interpMode=RCIM_Cubic, leaveTangentWeight=0.2293, tangentMode=RCTM_User, time=0.1, value=1}
         {arriveTangent=4.8659, interpMode=RCIM_Cubic, leaveTangent=4.8659, tangentMode=RCTM_User, time=0.75, value=2.75}
         {arriveTangent=24.818, interpMode=RCIM_Cubic, leaveTangent=24.818, tangentMode=RCTM_User, time=1, value=6}
        ]}}}))
   &DynamicMaterialParameters(
    Index 0 Param 1=&FloatFromCurve(
     FloatCurve=DataInterface() {propertyValues {Curve {
        defaultValue=0
        keys [
         {arriveTangent=1.5964, interpMode=RCIM_Cubic, leaveTangent=1.5964, tangentMode=RCTM_User, time=0, value=0}
         {arriveTangent=0.7982, interpMode=RCIM_Cubic, leaveTangent=0.7982, tangentMode=RCTM_User, time=1, value=1}
        ]
       }}}
     Scale Curve=0.75
    )
    Index 0 Param 3=-0.8
    Index 0 Param 4=&FloatFromCurve(FloatCurve=DataInterface() {propertyValues {Curve {keys [
         {interpMode=RCIM_Cubic, tangentMode=RCTM_User, time=0, value=1}
         {arriveTangent=-0.6219, interpMode=RCIM_Cubic, leaveTangent=-0.6219, tangentMode=RCTM_User, time=1, value=0}
        ]}}})
    Module.Param0WriteEnabled=true
    Module.Param1WriteEnabled=true
    Module.Param2WriteEnabled=true
    Module.Param3WriteEnabled=true
   )
  ]
  rendererClasses [&NiagaraRibbonRendererProperties]
  renderers [{rendererClass=&NiagaraRibbonRendererProperties}]
 }
 Trail_Dark {
  emitterUpdateScript [
   &EmitterState(Life Cycle Mode=::NewEnumerator1, Loop Behavior=::NewEnumerator1, Loop Duration=0.8, UseLoopDelay=false)
   &FN_SpawnAnimTrail()
  ]
  particleSpawnScript [
   &FN_InitializeAnimTrailBoneIndecies()
   &FN_InitializeAnimTrail(AnimTrailWidth=0.85, Color=<rgba 1 1 1 0.8>, Lifetime=0.38)
  ]
  particleUpdateScript [
   &ParticleState()
   &FN_UpdateAnimTrail(SkeletalMesh=$User.SkeletalMesh)
   &ScaleColor(
    Scale Alpha=&FloatFromCurve(FloatCurve=DataInterface() {propertyValues {Curve {keys [
         {arriveTangent=-2.3976, interpMode=RCIM_Cubic, leaveTangent=-2.3976, tangentMode=RCTM_User, time=0, value=1}
         {arriveTangent=0.0495, interpMode=RCIM_Cubic, leaveTangent=0.0495, tangentMode=RCTM_User, time=1, value=0}
        ]}}})
    Scale RGB=&MakeVectorFromLinearColorRGB(Color=&ColorFromCurve(ColorCurve=DataInterface() {propertyValues {
        AlphaCurve {keys []}
        BlueCurve {keys [<at 0.3218 0 RCIM_Cubic>,<at 1 0 RCIM_Cubic>]}
        GreenCurve {keys [<at 0.3218 0 RCIM_Cubic>,<at 1 0.001 RCIM_Cubic>]}
        RedCurve {keys [<at 0.3218 0 RCIM_Cubic>,<at 1 0.0443 RCIM_Cubic>]}
       }}))
    ScaleA=true
    ScaleRGB=true
   )
   &ScaleRibbonWidth(Ribbon Width Scale=&FloatFromCurve(FloatCurve=DataInterface() {propertyValues {Curve {
        defaultValue=1
        keys [{interpMode=RCIM_Cubic, tangentMode=RCTM_User, time=0, value=1}, {interpMode=RCIM_Cubic, tangentMode=RCTM_User, time=1, value=0}]
       }}}))
  ]
  rendererClasses [&NiagaraRibbonRendererProperties]
  renderers [{rendererClass=&NiagaraRibbonRendererProperties}]
 }
 Trail_Instrument_Color001 {
  emitterUpdateScript [
   &EmitterState(Life Cycle Mode=::NewEnumerator1, Loop Behavior=::NewEnumerator1, Loop Duration=0.22, UseLoopDelay=true)
   &FN_SpawnAnimTrail_v2()
  ]
  particleSpawnScript [
   /Game/Effects/Niagara/Modules/AnimTrail/FN_InitializeAnimTrail/FN_InitializeAnimTrailBoneIndecies(First Bone or Socket Index=1)
   /Game/Effects/Niagara/Modules/AnimTrail/FN_InitializeAnimTrail/FN_InitializeAnimTrail(AnimTrailWidth=1, Color=<rgba 0.015 0.0028 0.0029 1>, Lifetime=0.2)
   &SystemLocation(MaskSpawnGroup=false)
  ]
  particleUpdateScript [
   &ScaleRibbonWidth(Ribbon Width Scale=&FloatFromCurve(FloatCurve=DataInterface() {propertyValues {Curve {keys [
         {arriveTangent=-1.2579, interpMode=RCIM_Cubic, time=0, value=1}
         {arriveTangent=-0.9664, interpMode=RCIM_Cubic, leaveTangent=-0.9664, tangentMode=RCTM_User, time=1, value=0}
        ]}}}))
   &ScaleColor(
    Scale Alpha=&Multiply_Float(
     A=&FloatFromCurve(
      FloatCurve=DataInterface() {propertyValues {Curve {keys [
          {interpMode=RCIM_Cubic, leaveTangent=-2.3646, tangentMode=RCTM_Break, time=0, value=1}
          <at 0.5 0.0233>
          {arriveTangent=-0.2139, interpMode=RCIM_Cubic, tangentMode=RCTM_Break, time=1, value=0}
         ]}}}
      Scale Curve=&FloatFromCurve(
       FloatCurve=DataInterface() {propertyValues {Curve {keys [<at 0 1>,<at 0.497 1 RCIM_Cubic>,<at 1 1>]}}}
       CurveIndex=Linked() {linkedVariable {name=NPC.FortniteNPC.TimeOfDay_0-1, type=NiagaraFloat}}
      )
     )
     B=&MakeCustomFloatFromBool(Bool=$User.bIsFrontEnd, True Float=1, False Float=1)
    )
    Scale RGB=&MakeVectorFromLinearColorRGB(Color=$System.TOD Color)
    ScaleA=true
    ScaleRGB=true
   )
   &FN_UpdateAnimTrail_v2()
  ]
  rendererClasses [&NiagaraRibbonRendererProperties]
  renderers [{rendererClass=&NiagaraRibbonRendererProperties}]
 }
}
summary {
 systemName=NS_Pickaxe_Bass_PolkaSkate_Trail
 emitters [
  {emitterName=Trail_Instrument_Color001, rendererClasses [&NiagaraRibbonRendererProperties]}
  {emitterName=Ribbon_SmokeTrail, rendererClasses [&NiagaraRibbonRendererProperties]}
  {emitterName=Trail_Dark, rendererClasses [&NiagaraRibbonRendererProperties]}
  {emitterName=Petals_Falling, rendererClasses [&NiagaraSpriteRendererProperties]}
 ]
 userVariables [
  User.bIsFrontEnd {type=NiagaraBool}
  User.LandBurst {type=NiagaraFloat, defaultValue=0, description=""}
  User.RibbonIDOffset {type=NiagaraInt32}
  User.SkeletalMesh {
   type=NiagaraDataInterfaceSkeletalMesh
   defaultValue {
    struct=NiagaraExt_VariableValue_DataInterface
    value {dataInterface=&self:NiagaraDataInterfaceSkeletalMesh_0}
   }
  }
 ]
}
